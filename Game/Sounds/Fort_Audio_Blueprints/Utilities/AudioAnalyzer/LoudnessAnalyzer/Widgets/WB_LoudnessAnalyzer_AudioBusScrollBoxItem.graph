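# Loudness analyzer scroll box item widget that lays out an audio bus's name, routed submixes, and Loudness, Loudness Norm, Perceptual Energy and Timestamp readouts.
# On PreConstruct at design time, SetAudioBusRowSettingsDirect previews the StyleSettings struct by setting bus name and colors directly.
# Outside design time, PreConstruct instead broadcasts StyleSettings to the MVVM viewmodel.
# The direct preview leaves the routed-submixes text and color untouched.
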WidgetBlueprint /Game/Sounds/Fort_Audio_Blueprints/Utilities/AudioAnalyzer/LoudnessAnalyzer/Widgets/WB_LoudnessAnalyzer_AudioBusScrollBoxItem
graphs [
 EventGraph {nodes [
   n1 = Event PreConstruct()
   > n2 = Branch(Condition=@n1.IsDesignTime)
   > SetAudioBusRowSettingsDirect!(
    execute
    self
    AudioBusRowSettings:S Loudness Analyzer Audio Bus Row Settings Structure=Get StyleSettings() -> (StyleSettings:S Loudness Analyzer Audio Bus Row Settings Structure)
   ) -> (then)
   Set with Broadcast AudioBusRowSettings!(
    execute=@n2.else
    AudioBusRowSettings:S Loudness Analyzer Audio Bus Row Settings Structure=Get StyleSettings()
    self:MVVM Loudness Analyzer Audio Bus Scroll Box Item Object Reference=Get MVVM_LoudnessAnalyzer_AudioBusRow() -> (MVVM_LoudnessAnalyzer_AudioBusRow:MVVM Loudness Analyzer Audio Bus Scroll Box Item Object Reference)
   ) -> (then, Output_Get:S Loudness Analyzer Audio Bus Row Settings Structure)
  ]}
 SetAudioBusRowSettingsDirect {nodes [
   n1 = SetAudioBusRowSettingsDirect!() -> (then, AudioBusRowSettings:S Loudness Analyzer Audio Bus Row Settings Structure) {is=title K2Node_FunctionEntry}
   n2 = Break S Loudness Analyzer Audio Bus Row Settings!(S_LoudnessAnalyzer_AudioBusRowSettings:"S Loudness Analyzer Audio Bus Row Settings Structure (by ref)"=@n1) -> (
    AudioBusText_41_#1:Text
    RoutedSubmixesText_40_#2:Text
    RoutedSubmixesTextColor_44_#3:Slate Color Structure
    BorderColor_37_#4:Linear Color Structure
    TextColor_Label_25_#5:Slate Color Structure
    BackgroundColor_Label_Gradient_01_32_#6:Linear Color Structure
    BackgroundColor_Label_Gradient_02_33_#7:Linear Color Structure
    TextColor_HeaderRow_28_#8:Slate Color Structure
    BackgroundColor_HeaderRow_34_#9:Linear Color Structure
    TextColor_DataRow_30_#10:Slate Color Structure
    BackgroundColor_DataRow_35_#11:Linear Color Structure
   ) {is=title K2Node_BreakStruct}
   > Set Text(Text=@n2, self=Get AudioBusName() -> (AudioBusName:Text Block Object Reference))
   > Set BrushColor!(
    execute
    BrushColor:Linear Color Structure=@n2.BorderColor_37_#4
    self:Border Object Reference=Get BG_Border() -> (BG_Border:Border Object Reference)
   ) -> (then, Output_Get:Linear Color Structure)
   > Set ColorAndOpacity!(
    execute
    ColorAndOpacity:Slate Color Structure=@n2.TextColor_Label_25_#5
    self:Text Object Reference=Get AudioBusName()
   ) -> (then, Output_Get:Slate Color Structure)
   n3 = Get BG_Label() -> (BG_Label:Basic Gradient Fill Object Reference)
   > Set Color 1!(
    execute
    Color 1:Linear Color Structure=@n2.BackgroundColor_Label_Gradient_01_32_#6
    self:Basic Gradient Fill Object Reference=@n3
   ) -> (then, Output_Get:Linear Color Structure)
   > Set Color 2!(
    execute
    Color 2:Linear Color Structure=@n2.BackgroundColor_Label_Gradient_02_33_#7
    self:Basic Gradient Fill Object Reference=@n3
   ) -> (then, Output_Get:Linear Color Structure)
   > Set ColorAndOpacity(
    ColorAndOpacity=@n2.TextColor_HeaderRow_28_#8
    self=Get Header_Loudness() -> (Header_Loudness:Text Block Object Reference)
   )
   > Set ColorAndOpacity(
    ColorAndOpacity=@n2.TextColor_HeaderRow_28_#8
    self=Get Header_LoudnessNorm() -> (Header_LoudnessNorm:Text Block Object Reference)
   )
   > Set ColorAndOpacity(
    ColorAndOpacity=@n2.TextColor_HeaderRow_28_#8
    self=Get Header_PerceptualEnergy() -> (Header_PerceptualEnergy:Text Block Object Reference)
   )
   > Set ColorAndOpacity(
    ColorAndOpacity=@n2.TextColor_HeaderRow_28_#8
    self=Get Header_Timestamp() -> (Header_Timestamp:Text Block Object Reference)
   )
   > Set ColorAndOpacity(
    ColorAndOpacity:Linear Color Structure=@n2.BackgroundColor_HeaderRow_34_#9
    self:Image Object Reference=Get BG_HeaderRow() -> (BG_HeaderRow:Image Object Reference)
   ) -> (Output_Get:Linear Color Structure)
   > Set ColorAndOpacity(
    ColorAndOpacity:Slate Color Structure=@n2.TextColor_DataRow_30_#10
    self:Text Object Reference=Get Value_Loudness() -> (Value_Loudness:Text Block Object Reference)
   ) -> (Output_Get:Slate Color Structure)
   > Set ColorAndOpacity(
    ColorAndOpacity=@n2.TextColor_DataRow_30_#10
    self=Get Value_LoudnessNorm() -> (Value_LoudnessNorm:Text Block Object Reference)
   )
   > Set ColorAndOpacity(
    ColorAndOpacity=@n2.TextColor_DataRow_30_#10
    self=Get Value_PerceptualEnergy() -> (Value_PerceptualEnergy:Text Block Object Reference)
   )
   > Set ColorAndOpacity(
    ColorAndOpacity=@n2.TextColor_DataRow_30_#10
    self=Get Value_Timestamp() -> (Value_Timestamp:Text Block Object Reference)
   )
   > Set ColorAndOpacity(
    ColorAndOpacity:Linear Color Structure=@n2.BackgroundColor_DataRow_35_#11
    self:Image Object Reference=Get BG_DataRow() -> (BG_DataRow:Image Object Reference)
   ) -> (Output_Get:Linear Color Structure)
  ]}
]
variables [StyleSettings {
  type=S_LoudnessAnalyzer_AudioBusRowSettings
  default {
   audioBusText="<Audio Bus Name>"
   routedSubmixesText="<Routed Submixes>"
   backgroundColor_DataRow=<rgba 0.0156 0.0156 0.0156 1>
   backgroundColor_HeaderRow=<rgba 0 0 0 1>
   backgroundColor_Label_Gradient_01=<rgba 0 0.0936 0.2969 1>
   backgroundColor_Label_Gradient_02=<rgba 0 0.3734 0.5885 1>
   borderColor=<rgba 0 0 0 1>
   routedSubmixesTextColor {colorUseRule=UseColor_Specified, specifiedColor=<rgba 1 0.8192 0.412 1>}
   textColor_DataRow {colorUseRule=UseColor_Specified, specifiedColor=<rgba 0.5417 0.4076 0.108 1>}
   textColor_HeaderRow {colorUseRule=UseColor_Specified, specifiedColor=<rgba 0 0.6344 1 1>}
   textColor_Label {colorUseRule=UseColor_Specified, specifiedColor=<rgba 1 1 1 1>}
  }
 }]
widgets {
 info {inheritedWidgetCount=0, parentClass=&UserWidget, rootWidgetClass=&GridPanel, widgetCount=19}
 widgets [
  GridPanel_150 {widgetClassPath=&GridPanel, widget=GridPanel_150}
  BG_Border {
   widgetClassPath=&Border
   bIsVariable=true
   parent=GridPanel_150
   widget=BG_Border
   properties {
    background {
     bIsDynamicallyLoaded=false
     drawAs=Box
     imageType=NoImage
     mirroring=NoMirror
     resourceName=None
     resourceObject=null
     tiling=NoTile
     imageSize=<32 32>
     margin=<ltrb 0 0 0 0>
     outlineSettings {
      bUseBrushTransparency=false
      roundingType=HalfHeightRadius
      width=0
      color {colorUseRule=UseColor_Specified, specifiedColor=<rgba 0 0 0 0>}
      cornerRadii=<0 0 0 0>
     }
     tintColor {colorUseRule=UseColor_Specified, specifiedColor=<rgba 1 1 1 1>}
     uVRegion {bIsValid=false, max=<0 0>, min=<0 0>}
    }
    brushColor=<rgba 0 0 0 1>
   }
   slot {class=GridSlot, properties {columnSpan=6, rowSpan=6}}
  }
  BG_Label {
   widgetClassPath=/Game/UI/Foundation/Stats/BasicGradientFill
   bIsVariable=true
   parent=GridPanel_150
   widget=BG_Label
   properties {rotation=0.36, color 1=<rgba 0 0.0936 0.2969 1>, color 2=<rgba 0 0.3734 0.5885 1>}
   slot {class=GridSlot, properties {columnSpan=5, rowSpan=3, padding=<ltrb 5 5 0 0>}}
  }
  BG_HeaderRow {
   widgetClassPath=&Image
   bIsVariable=true
   parent=GridPanel_150
   widget=BG_HeaderRow
   properties {colorAndOpacity=<rgba 0 0 0 1>}
   slot {class=GridSlot, properties {columnSpan=5, layer=1, row=3, padding=<ltrb 5 0 0 0>}}
  }
  BG_DataRow {
   widgetClassPath=&Image
   bIsVariable=true
   parent=GridPanel_150
   widget=BG_DataRow
   properties {colorAndOpacity=<rgba 0.0156 0.0156 0.0156 1>}
   slot {class=GridSlot, properties {columnSpan=5, layer=1, row=4, padding=<ltrb 5 0 0 0>}}
  }
  AudioBusName {
   widgetClassPath=&UEFN_TextBlock
   bIsVariable=true
   parent=GridPanel_150
   widget=AudioBusName
   properties {
    text="<Audio Bus Name>"
    font {
     bForceMonospaced=false
     bMaterialIsStencil=false
     fontMaterial=null
     fontObject=&BurbankSmall
     letterSpacing=0
     monospacedWidth=1
     size=15
     skewAmount=0
     typefaceFontName=Black
     outlineSettings {
      bApplyOutlineToDropShadows=false
      bMiteredCorners=false
      bSeparateFillAlpha=false
      outlineMaterial=null
      outlineSize=0
      outlineColor=<rgba 0 0.0035 0.3854 1>
     }
    }
    shadowColorAndOpacity=<rgba 0 0 0 1>
   }
   slot {class=GridSlot, properties {columnSpan=2, row=1, verticalAlignment=VAlign_Center, padding=<ltrb 10 4 0 0>}}
  }
  RoutedSubmixes {
   widgetClassPath=&UEFN_TextBlock
   bIsVariable=true
   parent=GridPanel_150
   widget=RoutedSubmixes
   properties {
    text="<Routed Submixes>"
    colorAndOpacity {colorUseRule=UseColor_Specified, specifiedColor=<rgba 1 0.7074 0.3409 1>}
    font {
     bForceMonospaced=false
     bMaterialIsStencil=false
     fontMaterial=null
     fontObject=&BurbankSmall
     letterSpacing=0
     monospacedWidth=1
     size=12
     skewAmount=0
     typefaceFontName=Black
     outlineSettings {
      bApplyOutlineToDropShadows=false
      bMiteredCorners=false
      bSeparateFillAlpha=false
      outlineMaterial=null
      outlineSize=0
      outlineColor=<rgba 0 0.0035 0.3854 1>
     }
    }
    shadowColorAndOpacity=<rgba 0 0 0 1>
   }
   slot {class=GridSlot, properties {columnSpan=2, row=2, verticalAlignment=VAlign_Center, padding=<ltrb 10 0 0 4>}}
  }
  Spacer_179 {widgetClassPath=&Spacer, parent=GridPanel_150, widget=Spacer_179, slot {class=GridSlot, properties {column=4}}}
  Header_Loudness {
   widgetClassPath=&UEFN_TextBlock
   bIsVariable=true
   parent=GridPanel_150
   widget=Header_Loudness
   properties {
    text=Loudness
    colorAndOpacity {colorUseRule=UseColor_Specified, specifiedColor=<rgba 0 0.8247 1 1>}
    font {
     bForceMonospaced=false
     bMaterialIsStencil=false
     fontMaterial=null
     fontObject=&Roboto
     letterSpacing=200
     monospacedWidth=1
     size=9
     skewAmount=0
     typefaceFontName=Bold
     outlineSettings {
      bApplyOutlineToDropShadows=false
      bMiteredCorners=false
      bSeparateFillAlpha=false
      outlineMaterial=null
      outlineSize=0
      outlineColor=<rgba 0 0 0 1>
     }
    }
   }
   slot {class=GridSlot, properties {horizontalAlignment=HAlign_Left, layer=2, row=3, verticalAlignment=VAlign_Center, padding=<ltrb 10 5 0 5>}}
  }
  Header_LoudnessNorm {
   widgetClassPath=&UEFN_TextBlock
   bIsVariable=true
   parent=GridPanel_150
   widget=Header_LoudnessNorm
   properties {
    text=Loudness Norm
    colorAndOpacity {colorUseRule=UseColor_Specified, specifiedColor=<rgba 0 0.8247 1 1>}
    font {
     bForceMonospaced=false
     bMaterialIsStencil=false
     fontMaterial=null
     fontObject=&Roboto
     letterSpacing=200
     monospacedWidth=1
     size=9
     skewAmount=0
     typefaceFontName=Bold
     outlineSettings {
      bApplyOutlineToDropShadows=false
      bMiteredCorners=false
      bSeparateFillAlpha=false
      outlineMaterial=null
      outlineSize=0
      outlineColor=<rgba 0 0 0 1>
     }
    }
   }
   slot {
    class=GridSlot
    properties {column=1, horizontalAlignment=HAlign_Left, layer=2, row=3, verticalAlignment=VAlign_Center, padding=<ltrb 50 5 0 5>}
   }
  }
  Header_PerceptualEnergy {
   widgetClassPath=&UEFN_TextBlock
   bIsVariable=true
   parent=GridPanel_150
   widget=Header_PerceptualEnergy
   properties {
    text=Perceptual Energy
    colorAndOpacity {colorUseRule=UseColor_Specified, specifiedColor=<rgba 0 0.8247 1 1>}
    font {
     bForceMonospaced=false
     bMaterialIsStencil=false
     fontMaterial=null
     fontObject=&Roboto
     letterSpacing=200
     monospacedWidth=1
     size=9
     skewAmount=0
     typefaceFontName=Bold
     outlineSettings {
      bApplyOutlineToDropShadows=false
      bMiteredCorners=false
      bSeparateFillAlpha=false
      outlineMaterial=null
      outlineSize=0
      outlineColor=<rgba 0 0 0 1>
     }
    }
   }
   slot {
    class=GridSlot
    properties {column=2, horizontalAlignment=HAlign_Left, layer=2, row=3, verticalAlignment=VAlign_Center, padding=<ltrb 50 5 0 5>}
   }
  }
  Header_Timestamp {
   widgetClassPath=&UEFN_TextBlock
   bIsVariable=true
   parent=GridPanel_150
   widget=Header_Timestamp
   properties {
    text=Timestamp
    colorAndOpacity {colorUseRule=UseColor_Specified, specifiedColor=<rgba 0 0.8247 1 1>}
    font {
     bForceMonospaced=false
     bMaterialIsStencil=false
     fontMaterial=null
     fontObject=&Roboto
     letterSpacing=200
     monospacedWidth=1
     size=9
     skewAmount=0
     typefaceFontName=Bold
     outlineSettings {
      bApplyOutlineToDropShadows=false
      bMiteredCorners=false
      bSeparateFillAlpha=false
      outlineMaterial=null
      outlineSize=0
      outlineColor=<rgba 0 0 0 1>
     }
    }
   }
   slot {
    class=GridSlot
    properties {column=3, horizontalAlignment=HAlign_Left, layer=2, row=3, verticalAlignment=VAlign_Center, padding=<ltrb 45 5 70 5>}
   }
  }
  Value_Loudness {
   widgetClassPath=&UEFN_TextBlock
   bIsVariable=true
   parent=GridPanel_150
   widget=Value_Loudness
   properties {
    text=0.00
    colorAndOpacity {colorUseRule=UseColor_Specified, specifiedColor=<rgba 0.5417 0.5417 0.5417 1>}
    font {
     bForceMonospaced=true
     bMaterialIsStencil=false
     fontMaterial=null
     fontObject=&BurbankSmall
     letterSpacing=0
     monospacedWidth=1
     size=18
     skewAmount=0
     typefaceFontName=Medium
     outlineSettings {
      bApplyOutlineToDropShadows=false
      bMiteredCorners=false
      bSeparateFillAlpha=false
      outlineMaterial=null
      outlineSize=0
      outlineColor=<rgba 0 0 0 1>
     }
    }
   }
   slot {class=GridSlot, properties {horizontalAlignment=HAlign_Left, layer=2, row=4, verticalAlignment=VAlign_Center, padding=<ltrb 10 0 0 0>}}
  }
  Value_LoudnessNorm {
   widgetClassPath=&UEFN_TextBlock
   bIsVariable=true
   parent=GridPanel_150
   widget=Value_LoudnessNorm
   properties {
    text=0.00
    colorAndOpacity {colorUseRule=UseColor_Specified, specifiedColor=<rgba 0.5417 0.5417 0.5417 1>}
    font {
     bForceMonospaced=true
     bMaterialIsStencil=false
     fontMaterial=null
     fontObject=&BurbankSmall
     letterSpacing=0
     monospacedWidth=1
     size=18
     skewAmount=0
     typefaceFontName=Medium
     outlineSettings {
      bApplyOutlineToDropShadows=false
      bMiteredCorners=false
      bSeparateFillAlpha=false
      outlineMaterial=null
      outlineSize=0
      outlineColor=<rgba 0 0 0 1>
     }
    }
   }
   slot {
    class=GridSlot
    properties {column=1, horizontalAlignment=HAlign_Left, layer=2, row=4, verticalAlignment=VAlign_Center, padding=<ltrb 50 0 20 0>}
   }
  }
  Value_PerceptualEnergy {
   widgetClassPath=&UEFN_TextBlock
   bIsVariable=true
   parent=GridPanel_150
   widget=Value_PerceptualEnergy
   properties {
    text=0.00
    colorAndOpacity {colorUseRule=UseColor_Specified, specifiedColor=<rgba 0.5417 0.5417 0.5417 1>}
    font {
     bForceMonospaced=true
     bMaterialIsStencil=false
     fontMaterial=null
     fontObject=&BurbankSmall
     letterSpacing=0
     monospacedWidth=1
     size=18
     skewAmount=0
     typefaceFontName=Medium
     outlineSettings {
      bApplyOutlineToDropShadows=false
      bMiteredCorners=false
      bSeparateFillAlpha=false
      outlineMaterial=null
      outlineSize=0
      outlineColor=<rgba 0 0 0 1>
     }
    }
   }
   slot {
    class=GridSlot
    properties {column=2, horizontalAlignment=HAlign_Left, layer=2, row=4, verticalAlignment=VAlign_Center, padding=<ltrb 50 0 20 0>}
   }
  }
  Value_Timestamp {
   widgetClassPath=&UEFN_TextBlock
   bIsVariable=true
   parent=GridPanel_150
   widget=Value_Timestamp
   properties {
    text=0.00
    colorAndOpacity {colorUseRule=UseColor_Specified, specifiedColor=<rgba 0.5417 0.5417 0.5417 1>}
    font {
     bForceMonospaced=true
     bMaterialIsStencil=false
     fontMaterial=null
     fontObject=&BurbankSmall
     letterSpacing=0
     monospacedWidth=1
     size=18
     skewAmount=0
     typefaceFontName=Medium
     outlineSettings {
      bApplyOutlineToDropShadows=false
      bMiteredCorners=false
      bSeparateFillAlpha=false
      outlineMaterial=null
      outlineSize=0
      outlineColor=<rgba 0 0 0 1>
     }
    }
   }
   slot {
    class=GridSlot
    properties {column=3, horizontalAlignment=HAlign_Left, layer=2, row=4, verticalAlignment=VAlign_Center, padding=<ltrb 45 0 70 0>}
   }
  }
  Spacer_468 {
   widgetClassPath=&Spacer
   parent=GridPanel_150
   widget=Spacer_468
   properties {size=<5 1>}
   slot {class=GridSlot, properties {column=5, rowSpan=5}}
  }
  Spacer {
   widgetClassPath=&Spacer
   parent=GridPanel_150
   widget=Spacer
   properties {size=<379.893 1>}
   slot {class=GridSlot, properties {column=3, rowSpan=5}}
  }
  Spacer_356 {
   widgetClassPath=&Spacer
   parent=GridPanel_150
   widget=Spacer_356
   properties {size=<1 4>}
   slot {class=GridSlot, properties {row=5, padding=<ltrb 0 0 9 0>}}
  }
 ]
}
